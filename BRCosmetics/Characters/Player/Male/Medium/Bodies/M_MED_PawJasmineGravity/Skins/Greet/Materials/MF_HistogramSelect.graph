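# Material function in a character skin's material folder that selects a band of values from a scalar texture input, like a histogram select.
# It takes Tex, Position and Range, scaling by the reciprocal of Range and using Min and Max around Position to build two opposing ramps.
# The minimum of the ramps gives an unclamped triangular mask, peaking at one at Position and reaching zero Range away on either side.
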MaterialFunction /BRCosmetics/Characters/Player/Male/Medium/Bodies/M_MED_PawJasmineGravity/Skins/Greet/Materials/MF_HistogramSelect
graph {nodes [
  n1 = FunctionInput() {bUsePreviewValueAsDefault=true, inputName=Range, inputType=FunctionInput_Scalar, sortPriority=2, previewValue=<0.7 0 0 1>}
  n2 = Divide(A=Constant() {r=1}, B=@n1)
  n3 = FunctionInput(Preview=Noise!(World Position, FilterWidth) {scale=0.05}) {bUsePreviewValueAsDefault=true, inputName=Tex, inputType=FunctionInput_Scalar, previewValue=<1 0 0 1>}
  n4 = FunctionInput() {bUsePreviewValueAsDefault=true, inputName=Position, inputType=FunctionInput_Scalar, sortPriority=1, previewValue=<0.2 0 0 1>}
  n5 = Multiply(A=@n2, B=Subtract(A=Constant() {r=1}, B=Max(A=@n3, B=@n4)))
  n6 = Constant() {r=1}
  n7 = Subtract(A=Multiply(A=Subtract(A=@n6, B=@n4), B=@n2), B=Constant() {r=1})
  n8 = Multiply(A=@n2, B=Min(A=@n3, B=@n4))
  n9 = Subtract(A=Multiply(A=@n2, B=@n4), B=Constant() {r=1})
  FunctionOutput(None=Min(A=Subtract(A=@n5, B=@n7), B=Subtract(A=@n8, B=@n9)))
  TextureCoordinate()
 ]}
