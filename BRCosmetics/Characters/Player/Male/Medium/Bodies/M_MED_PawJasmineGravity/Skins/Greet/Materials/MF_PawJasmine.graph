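# Material function for the PawJasmine Greet player skin, layering an animated, glitchy tech camo effect over incoming material attributes.
# Parallax-mapped camo, stabilized screen-space scanlines, pulsing quantized dot grids and Fresnel rims feed the emissive, with stepped time and sine-driven flicker.
# Masks confine the effect, which also alters base color and roughness, uses temporal dithering and offsets vertices along normals.
# It outputs the modified attributes, camo area and roughness.
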MaterialFunction /BRCosmetics/Characters/Player/Male/Medium/Bodies/M_MED_PawJasmineGravity/Skins/Greet/Materials/MF_PawJasmine
graph {expressions {
  MaterialExpressionAdd_0 {
   class=MaterialExpressionAdd
   inputs [
    A {from=MaterialExpressionGetMaterialAttributes_2, output=WorldPositionOffset}
    B {from=MaterialExpressionMultiply_0}
   ]
  }
  MaterialExpressionAdd_1 {
   class=MaterialExpressionAdd
   inputs [
    A {from=MaterialExpressionAppendVector_0}
    B {from=MaterialExpressionMultiply_7}
   ]
  }
  MaterialExpressionAdd_16 {
   class=MaterialExpressionAdd
   inputs [
    A {from=MaterialExpressionComponentMask_2}
    B {from=MaterialExpressionMultiply_26}
   ]
  }
  MaterialExpressionAdd_17 {
   class=MaterialExpressionAdd
   inputs [
    A {from=MaterialExpressionMultiply_28}
    B {from=MaterialExpressionMaterialFunctionCall_24, output=Result}
   ]
  }
  MaterialExpressionAdd_19 {
   class=MaterialExpressionAdd
   inputs [
    A {from=MaterialExpressionNamedRerouteDeclaration_11}
    B {from=MaterialExpressionMultiply_59}
   ]
  }
  MaterialExpressionAdd_2 {
   class=MaterialExpressionAdd
   inputs [
    A {from=MaterialExpressionPanner_0}
    B {from=MaterialExpressionMaterialFunctionCall_4, output=Result}
   ]
  }
  MaterialExpressionAdd_3 {
   class=MaterialExpressionAdd
   inputs [
    A {from=MaterialExpressionScalarParameter_12}
    B {from=MaterialExpressionScalarParameter_13}
   ]
  }
  MaterialExpressionAdd_4 {
   class=MaterialExpressionAdd
   inputs [
    A {from=MaterialExpressionMax_4}
    B {from=MaterialExpressionMultiply_52}
   ]
  }
  MaterialExpressionAdd_5 {
   class=MaterialExpressionAdd
   inputs [
    A {from=MaterialExpressionTime_10}
    B {from=MaterialExpressionMultiply_36}
   ]
  }
  MaterialExpressionAdd_6 {
   class=MaterialExpressionAdd
   inputs [
    A {from=MaterialExpressionNamedRerouteDeclaration_11}
    B {from=MaterialExpressionMultiply_13}
   ]
  }
  MaterialExpressionAdd_8 {
   class=MaterialExpressionAdd
   inputs [
    A {from=MaterialExpressionMultiply_14}
    B {from=MaterialExpressionMultiply_17}
   ]
  }
  MaterialExpressionAdd_9 {
   class=MaterialExpressionAdd
   inputs [
    A {from=MaterialExpressionNamedRerouteDeclaration_11}
    B {from=MaterialExpressionMultiply_15}
   ]
  }
  MaterialExpressionAppendVector_0 {
   class=MaterialExpressionAppendVector
   inputs [
    A {from=MaterialExpressionTextureSampleParameter2D_0, output=A}
    B {from=MaterialExpressionTextureSampleParameter2D_0, output=A}
   ]
  }
  MaterialExpressionCeil_0 {class=MaterialExpressionCeil, inputs [None {from=MaterialExpressionMultiply_2, output=""}], properties {}}
  MaterialExpressionComponentMask_2 {
   class=MaterialExpressionComponentMask
   inputs [None {from=MaterialExpressionMaterialFunctionCall_25, output=StabilizedUVs}]
   properties {g=true}
  }
  MaterialExpressionConstant4Vector_0 {class=MaterialExpressionConstant4Vector, inputs [], properties {constant=<rgba 1 0 0 1>}}
  MaterialExpressionConstant_0 {class=MaterialExpressionConstant, properties {r=2}}
  MaterialExpressionConstant_1 {class=MaterialExpressionConstant, properties {r=1}}
  MaterialExpressionConstant_2 {class=MaterialExpressionConstant, properties {r=1}}
  MaterialExpressionConstant_3 {class=MaterialExpressionConstant, properties {r=2}}
  MaterialExpressionConstant_4 {class=MaterialExpressionConstant, properties {r=8}}
  MaterialExpressionConvert_0 {
   class=MaterialExpressionConvert
   inputs [
    None {from=MaterialExpressionScalarParameter_37}
    None {from=MaterialExpressionScalarParameter_38}
   ]
   properties {
    nodeName=Make Float2
    convertInputs [
     {
      type=Scalar
      expressionInput {
       expression=&self:MaterialExpressionScalarParameter_37
       inputName=None
       mask=0
       maskA=0
       maskB=0
       maskG=0
       maskR=0
       outputIndex=0
      }
     }
     {
      type=Scalar
      expressionInput {
       expression=&self:MaterialExpressionScalarParameter_38
       inputName=None
       mask=0
       maskA=0
       maskB=0
       maskG=0
       maskR=0
       outputIndex=0
      }
     }
    ]
    convertOutputs [{type=Vector2}]
   }
  }
  MaterialExpressionDesaturation_0 {
   class=MaterialExpressionDesaturation
   inputs [
    None {from=MaterialExpressionNamedRerouteDeclaration_28, output=""}
    Fraction {from="", output=""}
   ]
   properties {}
  }
  MaterialExpressionDivide_0 {
   class=MaterialExpressionDivide
   inputs [
    A {from=MaterialExpressionFloor_0}
    B {from=MaterialExpressionScalarParameter_28}
   ]
  }
  MaterialExpressionDivide_1 {
   class=MaterialExpressionDivide
   inputs [
    A {from=MaterialExpressionCeil_0}
    B {from=MaterialExpressionScalarParameter_17}
   ]
  }
  MaterialExpressionDivide_3 {
   class=MaterialExpressionDivide
   inputs [
    A {from=MaterialExpressionScalarParameter_28}
    B {from=""}
   ]
   properties {constB=50}
  }
  MaterialExpressionFloor_0 {class=MaterialExpressionFloor, inputs [None {from=MaterialExpressionMultiply_29, output=""}], properties {}}
  MaterialExpressionFrac_0 {class=MaterialExpressionFrac, inputs [None {from=MaterialExpressionLinearInterpolate_6, output=""}], properties {}}
  MaterialExpressionFrac_2 {class=MaterialExpressionFrac, inputs [None {from=MaterialExpressionMultiply_20, output=""}], properties {}}
  MaterialExpressionFresnel_0 {
   class=MaterialExpressionFresnel
   inputs [
    ExponentIn {from=MaterialExpressionScalarParameter_0, output=""}
    BaseReflectFractionIn {from=MaterialExpressionScalarParameter_1, output=""}
    Normal {from=MaterialExpressionTransform_1, output=""}
   ]
   properties {baseReflectFraction=0}
  }
  MaterialExpressionFresnel_1 {
   class=MaterialExpressionFresnel
   inputs [
    ExponentIn {from=MaterialExpressionScalarParameter_58, output=""}
    BaseReflectFractionIn {from="", output=""}
    Normal {from=MaterialExpressionTransform_1, output=""}
   ]
   properties {baseReflectFraction=0}
  }
  MaterialExpressionFresnel_2 {
   class=MaterialExpressionFresnel
   inputs [
    ExponentIn {from=MaterialExpressionScalarParameter_8, output=""}
    BaseReflectFractionIn {from="", output=""}
    Normal {from="", output=""}
   ]
   properties {baseReflectFraction=0}
  }
  MaterialExpressionFunctionInput_0 {
   class=MaterialExpressionFunctionInput
   inputs [Preview {from="", output=""}]
   properties {inputName=Tech Art Mask, inputType=FunctionInput_Scalar}
  }
  MaterialExpressionFunctionInput_1 {
   class=MaterialExpressionFunctionInput
   inputs [Preview {from="", output=""}]
   properties {inputName=Result, inputType=FunctionInput_MaterialAttributes, sortPriority=10}
  }
  MaterialExpressionFunctionOutput_1 {
   class=MaterialExpressionFunctionOutput
   inputs [None {from=MaterialExpressionStaticSwitchParameter_1, output=""}]
   properties {outputName="MA (Out)"}
  }
  MaterialExpressionFunctionOutput_2 {
   class=MaterialExpressionFunctionOutput
   inputs [None {from=MaterialExpressionNamedRerouteDeclaration_16, output=""}]
   properties {outputName=Camo Area}
  }
  MaterialExpressionFunctionOutput_3 {
   class=MaterialExpressionFunctionOutput
   inputs [None {from=MaterialExpressionNamedRerouteUsage_6, output=""}]
   properties {outputName=Roughness}
  }
  MaterialExpressionGetMaterialAttributes_0 {
   class=MaterialExpressionGetMaterialAttributes
   inputs [None {from=MaterialExpressionFunctionInput_1, output=""}]
   properties {attributeGetTypes [
     #1
     #2
     #3
     #4
     #5
    ]}
  }
  MaterialExpressionGetMaterialAttributes_2 {
   class=MaterialExpressionGetMaterialAttributes
   inputs [None {from=MaterialExpressionStaticSwitchParameter_0, output=""}]
   properties {attributeGetTypes [#6]}
  }
  MaterialExpressionLinearInterpolate_0 {
   class=MaterialExpressionLinearInterpolate
   inputs [
    A {from=MaterialExpressionNamedRerouteUsage_33}
    B {from=MaterialExpressionNamedRerouteUsage_32}
    Alpha {from=MaterialExpressionNamedRerouteUsage_5}
   ]
  }
  MaterialExpressionLinearInterpolate_1 {
   class=MaterialExpressionLinearInterpolate
   inputs [
    A {from=MaterialExpressionNamedRerouteUsage_34}
    B {from=""}
    Alpha {from=MaterialExpressionNamedRerouteDeclaration_16}
   ]
   properties {constB=2}
  }
  MaterialExpressionLinearInterpolate_11 {
   class=MaterialExpressionLinearInterpolate
   inputs [
    A {from=""}
    B {from=MaterialExpressionNamedRerouteUsage_13}
    Alpha {from=MaterialExpressionScalarParameter_56}
   ]
   properties {constA=1}
  }
  MaterialExpressionLinearInterpolate_12 {
   class=MaterialExpressionLinearInterpolate
   inputs [
    A {from=MaterialExpressionTextureCoordinate_1}
    B {from=MaterialExpressionTextureCoordinate_4}
    Alpha {from=MaterialExpressionScalarParameter_46}
   ]
  }
  MaterialExpressionLinearInterpolate_15 {
   class=MaterialExpressionLinearInterpolate
   inputs [
    A {from=MaterialExpressionLinearInterpolate_5}
    B {from=""}
    Alpha {from=MaterialExpressionScalarParameter_76}
   ]
   properties {constA=1}
  }
  MaterialExpressionLinearInterpolate_16 {
   class=MaterialExpressionLinearInterpolate
   inputs [
    A {from=MaterialExpressionNamedRerouteUsage_22}
    B {from=MaterialExpressionVectorParameter_0, output=RGB}
    Alpha {from=MaterialExpressionSaturate_4}
   ]
  }
  MaterialExpressionLinearInterpolate_2 {
   class=MaterialExpressionLinearInterpolate
   inputs [
    A {from=MaterialExpressionNamedRerouteUsage_6}
    B {from=MaterialExpressionScalarParameter_61}
    Alpha {from=MaterialExpressionNamedRerouteDeclaration_16}
   ]
   properties {constB=2}
  }
  MaterialExpressionLinearInterpolate_4 {
   class=MaterialExpressionLinearInterpolate
   inputs [
    A {from=MaterialExpressionNamedRerouteUsage_15}
    B {from=""}
    Alpha {from=MaterialExpressionNamedRerouteDeclaration_16}
   ]
   properties {constB=0}
  }
  MaterialExpressionLinearInterpolate_5 {
   class=MaterialExpressionLinearInterpolate
   inputs [
    A {from=""}
    B {from=MaterialExpressionSaturate_2}
    Alpha {from=MaterialExpressionScalarParameter_23}
   ]
   properties {constA=1}
  }
  MaterialExpressionLinearInterpolate_6 {
   class=MaterialExpressionLinearInterpolate
   inputs [
    A {from=MaterialExpressionScalarParameter_22}
    B {from=""}
    Alpha {from=MaterialExpressionMultiply_38}
   ]
  }
  MaterialExpressionLinearInterpolate_8 {
   class=MaterialExpressionLinearInterpolate
   inputs [
    A {from=MaterialExpressionGetMaterialAttributes_2, output=WorldPositionOffset}
    B {from=MaterialExpressionAdd_0}
    Alpha {from=MaterialExpressionNamedRerouteUsage_29}
   ]
  }
  MaterialExpressionMaterialFunctionCall_10 {
   class=MaterialExpressionMaterialFunctionCall
   inputs [
    Tex {from=MaterialExpressionTextureSample_3, output=G}
    Position {from=MaterialExpressionMaterialFunctionCall_12, output=Remapped Sine}
    Range {from=MaterialExpressionConstant_1}
   ]
   properties {materialFunction=&/MF_HistogramSelect}
  }
  MaterialExpressionMaterialFunctionCall_11 {
   class=MaterialExpressionMaterialFunctionCall
   inputs [
    X {from=""}
    Y {from=MaterialExpressionScalarParameter_5}
   ]
   properties {materialFunction=&MakeFloat2}
  }
  MaterialExpressionMaterialFunctionCall_12 {
   class=MaterialExpressionMaterialFunctionCall
   inputs [
    Sine Phase {from=MaterialExpressionMultiply_45}
    Value 1 {from=""}
    Value 2 {from=""}
   ]
   properties {materialFunction=/Engine/Functions/Engine_MaterialFunctions02/Utility/Sine_Remapped}
  }
  MaterialExpressionMaterialFunctionCall_14 {
   class=MaterialExpressionMaterialFunctionCall
   inputs [
    X {from=MaterialExpressionConstant_2}
    Y {from=MaterialExpressionConstant_2}
   ]
   properties {materialFunction=&MakeFloat2}
  }
  MaterialExpressionMaterialFunctionCall_15 {
   class=MaterialExpressionMaterialFunctionCall
   inputs [
    "Param (0-1)" {from=MaterialExpressionSubtract_3}
    Tiles {from=MaterialExpressionConstant_0}
    Apply Offset {from=MaterialExpressionStaticBoolParameter_0}
   ]
   properties {materialFunction=/Game/Environments/Papaya/Materials/MaterialFunctions/MF_Papaya_Quantize_Scalar}
  }
  MaterialExpressionMaterialFunctionCall_16 {
   class=MaterialExpressionMaterialFunctionCall
   inputs [
    Alpha Threshold {from=MaterialExpressionMultiply_5}
    Random {from=""}
   ]
   properties {materialFunction=&DitherTemporalAA}
  }
  MaterialExpressionMaterialFunctionCall_17 {
   class=MaterialExpressionMaterialFunctionCall
   inputs [
    Heightmap Texture {from=MaterialExpressionTextureObject_0}
    Height Ratio {from=MaterialExpressionScalarParameter_6}
    Min Steps {from=MaterialExpressionConstant_4}
    Max Steps {from=MaterialExpressionConstant_4}
    UVs {from=MaterialExpressionAdd_8}
    Heightmap Channel {from=MaterialExpressionConstant4Vector_0, output=RGBA}
    Reference Plane {from=""}
    ----------------------------- {from=""}
    Use World Coordinates {from=""}
    Specify Manual Texture Size {from=""}
    Manual Texture Size {from=""}
    ------------------------------ {from=""}
    "Render Shadows (Occlusion Mapping)" {from=""}
    Light Vector {from=""}
    Shadow Steps {from=""}
    Shadow Penumbra {from=""}
    Transform To VertexNormal {from=""}
   ]
   properties {materialFunction=/Engine/Functions/Engine_MaterialFunctions01/Texturing/ParallaxOcclusionMapping}
  }
  MaterialExpressionMaterialFunctionCall_2 {
   class=MaterialExpressionMaterialFunctionCall
   inputs [
    X {from=MaterialExpressionScalarParameter_7}
    Y {from=MaterialExpressionScalarParameter_14}
   ]
   properties {materialFunction=&MakeFloat2}
  }
  MaterialExpressionMaterialFunctionCall_20 {
   class=MaterialExpressionMaterialFunctionCall
   inputs [
    "Param (0-1)" {from=MaterialExpressionMaterialFunctionCall_7, output=Result}
    Tiles {from=MaterialExpressionConstant_3}
    Apply Offset {from=""}
   ]
   properties {materialFunction=&MF_Papaya_Quantize_Scalar}
  }
  MaterialExpressionMaterialFunctionCall_21 {
   class=MaterialExpressionMaterialFunctionCall
   inputs [
    Blend {from=MaterialExpressionNamedRerouteUsage_19}
    Base {from=MaterialExpressionMultiply_1}
   ]
   properties {materialFunction=/Engine/Functions/Engine_MaterialFunctions03/Blends/Blend_Screen}
  }
  MaterialExpressionMaterialFunctionCall_22 {
   class=MaterialExpressionMaterialFunctionCall
   inputs [
    UV {from=MaterialExpressionNamedRerouteUsage_14}
    NumDotsX {from=MaterialExpressionScalarParameter_28}
    NumDotsY {from=MaterialExpressionScalarParameter_28}
    SharpRadius {from=""}
    GlowDensity {from=""}
    GlowRadius {from=""}
    GlowMin {from=""}
    GlowMax {from=""}
   ]
   properties {materialFunction=/AudioWidgets/AudioMaterialSlate/Functions/MF_Dots}
  }
  MaterialExpressionMaterialFunctionCall_23 {
   class=MaterialExpressionMaterialFunctionCall
   inputs [
    "Param (0-1)" {from=MaterialExpressionMaterialFunctionCall_10, output=Result}
    Tiles {from=MaterialExpressionConstant_3}
    Apply Offset {from=""}
   ]
   properties {materialFunction=&MF_Papaya_Quantize_Scalar}
  }
  MaterialExpressionMaterialFunctionCall_24 {
   class=MaterialExpressionMaterialFunctionCall
   inputs [
    X {from=""}
    Y {from=MaterialExpressionMultiply_30}
   ]
   properties {materialFunction=&MakeFloat2}
  }
  MaterialExpressionMaterialFunctionCall_25 {
   class=MaterialExpressionMaterialFunctionCall
   inputs [
    Position {from=""}
    UVRadialScale {from=""}
   ]
   properties {materialFunction=/Game/Characters/Player/Common/MaterialFunctions/Superlevel/CharacterStabilizedScreenSpaceUVs}
  }
  MaterialExpressionMaterialFunctionCall_26 {
   class=MaterialExpressionMaterialFunctionCall
   inputs [
    X {from=MaterialExpressionScalarParameter_9}
    Y {from=""}
   ]
   properties {materialFunction=&MakeFloat2}
  }
  MaterialExpressionMaterialFunctionCall_29 {
   class=MaterialExpressionMaterialFunctionCall
   inputs [
    Tex {from=MaterialExpressionMultiply_47}
    Position {from=MaterialExpressionScalarParameter_54}
    Range {from=MaterialExpressionScalarParameter_55}
   ]
   properties {materialFunction=&MF_HistogramSelect}
  }
  MaterialExpressionMaterialFunctionCall_4 {
   class=MaterialExpressionMaterialFunctionCall
   inputs [
    X {from=""}
    Y {from=MaterialExpressionScalarParameter_45}
   ]
   properties {materialFunction=&MakeFloat2}
  }
  MaterialExpressionMaterialFunctionCall_40 {
   class=MaterialExpressionMaterialFunctionCall
   inputs [
    ObjectPivotLocation {from=""}
    ObjectPositionOffset {from=""}
   ]
   properties {materialFunction=/Game/Characters/Player/Common/MaterialFunctions/Superlevel/ImprovedStabilizedScreenSpaceUVs}
  }
  MaterialExpressionMaterialFunctionCall_41 {
   class=MaterialExpressionMaterialFunctionCall
   inputs [
    Fade Length {from=MaterialExpressionScalarParameter_68}
    Fade Offset {from=MaterialExpressionScalarParameter_67}
    For Vertex Shader {from=""}
   ]
   properties {materialFunction=&CameraDepthFade}
  }
  MaterialExpressionMaterialFunctionCall_5 {
   class=MaterialExpressionMaterialFunctionCall
   inputs [
    X {from=MaterialExpressionScalarParameter_4}
    Y {from=""}
   ]
   properties {materialFunction=&MakeFloat2}
  }
  MaterialExpressionMaterialFunctionCall_6 {
   class=MaterialExpressionMaterialFunctionCall
   inputs [Multiplier {from=MaterialExpressionScalarParameter_33}]
   properties {materialFunction=&Pi}
  }
  MaterialExpressionMaterialFunctionCall_7 {
   class=MaterialExpressionMaterialFunctionCall
   inputs [
    Tex {from=MaterialExpressionTextureSample_3, output=R}
    Position {from=MaterialExpressionMaterialFunctionCall_12, output=Remapped Sine}
    Range {from=MaterialExpressionConstant_1}
   ]
   properties {materialFunction=&MF_HistogramSelect}
  }
  MaterialExpressionMaterialFunctionCall_8 {
   class=MaterialExpressionMaterialFunctionCall
   inputs [
    UVs {from=MaterialExpressionMultiply_3}
    Rotation Center {from=""}
    "Rotation Angle (0-1)" {from=MaterialExpressionScalarParameter_36}
   ]
   properties {materialFunction=&CustomRotator}
  }
  MaterialExpressionMaterialFunctionCall_9 {
   class=MaterialExpressionMaterialFunctionCall
   inputs [
    X {from=""}
    Y {from=MaterialExpressionScalarParameter_9}
   ]
   properties {materialFunction=&MakeFloat2}
  }
  MaterialExpressionMax_1 {
   class=MaterialExpressionMax
   inputs [
    A {from=MaterialExpressionOneMinus_1}
    B {from=MaterialExpressionScalarParameter_1}
   ]
  }
  MaterialExpressionMax_2 {
   class=MaterialExpressionMax
   inputs [
    A {from=MaterialExpressionNamedRerouteUsage_9}
    B {from=""}
   ]
   properties {constB=0.6}
  }
  MaterialExpressionMax_3 {
   class=MaterialExpressionMax
   inputs [
    A {from=MaterialExpressionMultiply_24}
    B {from=MaterialExpressionMultiply_23}
   ]
  }
  MaterialExpressionMax_4 {
   class=MaterialExpressionMax
   inputs [
    A {from=MaterialExpressionMaterialFunctionCall_21, output=Result}
    B {from=MaterialExpressionNamedRerouteUsage_18}
   ]
  }
  MaterialExpressionMax_5 {
   class=MaterialExpressionMax
   inputs [
    A {from=MaterialExpressionMultiply_55}
    B {from=MaterialExpressionMultiply_56}
   ]
  }
  MaterialExpressionMax_6 {
   class=MaterialExpressionMax
   inputs [
    A {from=MaterialExpressionMultiply_46}
    B {from=MaterialExpressionMultiply_35}
   ]
  }
  MaterialExpressionMax_9 {
   class=MaterialExpressionMax
   inputs [
    A {from=MaterialExpressionTextureSample_10, output=R}
    B {from=MaterialExpressionTextureSample_12, output=R}
   ]
  }
  MaterialExpressionMultiply_0 {
   class=MaterialExpressionMultiply
   inputs [
    A {from=MaterialExpressionVertexNormalWS_0}
    B {from=MaterialExpressionScalarParameter_3}
   ]
  }
  MaterialExpressionMultiply_1 {
   class=MaterialExpressionMultiply
   inputs [
    A {from=MaterialExpressionMax_1}
    B {from=MaterialExpressionNamedRerouteUsage_20}
   ]
  }
  MaterialExpressionMultiply_10 {
   class=MaterialExpressionMultiply
   inputs [
    A {from=MaterialExpressionSine_1}
    B {from=MaterialExpressionSine_2}
   ]
  }
  MaterialExpressionMultiply_11 {
   class=MaterialExpressionMultiply
   inputs [
    A {from=MaterialExpressionStep_2}
    B {from=MaterialExpressionStep_0}
   ]
  }
  MaterialExpressionMultiply_12 {
   class=MaterialExpressionMultiply
   inputs [
    A {from=MaterialExpressionNamedRerouteUsage_8}
    B {from=MaterialExpressionLinearInterpolate_11}
   ]
  }
  MaterialExpressionMultiply_13 {
   class=MaterialExpressionMultiply
   inputs [
    A {from=MaterialExpressionTime_4}
    B {from=MaterialExpressionMaterialFunctionCall_11, output=Result}
   ]
  }
  MaterialExpressionMultiply_14 {
   class=MaterialExpressionMultiply
   inputs [
    A {from=MaterialExpressionAdd_6}
    B {from=MaterialExpressionMaterialFunctionCall_2, output=Result}
   ]
  }
  MaterialExpressionMultiply_15 {
   class=MaterialExpressionMultiply
   inputs [
    A {from=MaterialExpressionMaterialFunctionCall_9, output=Result}
    B {from=MaterialExpressionTime_5}
   ]
  }
  MaterialExpressionMultiply_16 {
   class=MaterialExpressionMultiply
   inputs [
    A {from=MaterialExpressionAdd_9}
    B {from=MaterialExpressionScalarParameter_10}
   ]
  }
  MaterialExpressionMultiply_17 {
   class=MaterialExpressionMultiply
   inputs [
    A {from=MaterialExpressionScalarParameter_11}
    B {from=MaterialExpressionMax_9}
   ]
  }
  MaterialExpressionMultiply_18 {
   class=MaterialExpressionMultiply
   inputs [
    A {from=MaterialExpressionMultiply_61}
    B {from=MaterialExpressionMultiply_19}
   ]
  }
  MaterialExpressionMultiply_19 {
   class=MaterialExpressionMultiply
   inputs [
    A {from=MaterialExpressionVectorParameter_1, output=RGB}
    B {from=MaterialExpressionScalarParameter_57}
   ]
  }
  MaterialExpressionMultiply_2 {
   class=MaterialExpressionMultiply
   inputs [
    A {from=MaterialExpressionTime_11}
    B {from=MaterialExpressionScalarParameter_17}
   ]
  }
  MaterialExpressionMultiply_20 {
   class=MaterialExpressionMultiply
   inputs [
    A {from=MaterialExpressionAdd_16}
    B {from=MaterialExpressionScalarParameter_18}
   ]
  }
  MaterialExpressionMultiply_21 {
   class=MaterialExpressionMultiply
   inputs [
    A {from=MaterialExpressionSmoothStep_4}
    B {from=MaterialExpressionScalarParameter_21}
   ]
  }
  MaterialExpressionMultiply_22 {
   class=MaterialExpressionMultiply
   inputs [
    A {from=MaterialExpressionTextureSample_2, output=RGB}
    B {from=MaterialExpressionScalarParameter_41}
   ]
  }
  MaterialExpressionMultiply_23 {
   class=MaterialExpressionMultiply
   inputs [
    A {from=MaterialExpressionMultiply_21}
    B {from=MaterialExpressionVectorParameter_2, output=RGB}
   ]
  }
  MaterialExpressionMultiply_24 {
   class=MaterialExpressionMultiply
   inputs [
    A {from=MaterialExpressionSmoothStep_3}
    B {from=MaterialExpressionVectorParameter_3, output=RGB}
   ]
  }
  MaterialExpressionMultiply_25 {
   class=MaterialExpressionMultiply
   inputs [
    A {from=MaterialExpressionMax_3}
    B {from=MaterialExpressionScalarParameter_25}
   ]
  }
  MaterialExpressionMultiply_26 {
   class=MaterialExpressionMultiply
   inputs [
    A {from=MaterialExpressionTime_7}
    B {from=MaterialExpressionScalarParameter_26}
   ]
  }
  MaterialExpressionMultiply_27 {
   class=MaterialExpressionMultiply
   inputs [
    A {from=MaterialExpressionTextureSampleParameter2D_1, output=RGB}
    B {from=MaterialExpressionScalarParameter_60}
   ]
  }
  MaterialExpressionMultiply_28 {
   class=MaterialExpressionMultiply
   inputs [
    A {from=MaterialExpressionScalarParameter_30}
    B {from=MaterialExpressionDivide_0}
   ]
  }
  MaterialExpressionMultiply_29 {
   class=MaterialExpressionMultiply
   inputs [
    A {from=MaterialExpressionNamedRerouteUsage_12}
    B {from=MaterialExpressionScalarParameter_28}
   ]
  }
  MaterialExpressionMultiply_3 {
   class=MaterialExpressionMultiply
   inputs [
    A {from=MaterialExpressionNamedRerouteUsage_7}
    B {from=MaterialExpressionConvert_0}
   ]
  }
  MaterialExpressionMultiply_30 {
   class=MaterialExpressionMultiply
   inputs [
    A {from=MaterialExpressionTime_8}
    B {from=MaterialExpressionScalarParameter_32}
   ]
  }
  MaterialExpressionMultiply_31 {
   class=MaterialExpressionMultiply
   inputs [
    A {from=MaterialExpressionNamedRerouteUsage_14}
    B {from=MaterialExpressionMultiply_33}
   ]
  }
  MaterialExpressionMultiply_32 {
   class=MaterialExpressionMultiply
   inputs [
    A {from=MaterialExpressionTextureSample_0, output=R}
    B {from=MaterialExpressionScalarParameter_43}
   ]
  }
  MaterialExpressionMultiply_33 {
   class=MaterialExpressionMultiply
   inputs [
    A {from=MaterialExpressionScalarParameter_28}
    B {from=MaterialExpressionScalarParameter_16}
   ]
  }
  MaterialExpressionMultiply_34 {
   class=MaterialExpressionMultiply
   inputs [
    A {from=MaterialExpressionTextureSample_0, output=R}
    B {from=MaterialExpressionStep_1}
   ]
  }
  MaterialExpressionMultiply_35 {
   class=MaterialExpressionMultiply
   inputs [
    A {from=MaterialExpressionMultiply_34}
    B {from=MaterialExpressionScalarParameter_34}
   ]
  }
  MaterialExpressionMultiply_36 {
   class=MaterialExpressionMultiply
   inputs [
    A {from=MaterialExpressionScalarParameter_47}
    B {from=MaterialExpressionTextureSample_5, output=R}
   ]
  }
  MaterialExpressionMultiply_37 {
   class=MaterialExpressionMultiply
   inputs [
    A {from=MaterialExpressionMultiply_10}
    B {from=MaterialExpressionSine_0}
   ]
  }
  MaterialExpressionMultiply_38 {
   class=MaterialExpressionMultiply
   inputs [
    A {from=MaterialExpressionMultiply_37}
    B {from=MaterialExpressionSine_4}
   ]
  }
  MaterialExpressionMultiply_39 {
   class=MaterialExpressionMultiply
   inputs [
    A {from=MaterialExpressionTime_1}
    B {from=MaterialExpressionScalarParameter_31}
   ]
  }
  MaterialExpressionMultiply_4 {
   class=MaterialExpressionMultiply
   inputs [
    A {from=MaterialExpressionMultiply_44}
    B {from=MaterialExpressionMultiply_22}
   ]
  }
  MaterialExpressionMultiply_40 {
   class=MaterialExpressionMultiply
   inputs [
    A {from=MaterialExpressionTime_0}
    B {from=MaterialExpressionScalarParameter_31}
   ]
  }
  MaterialExpressionMultiply_41 {
   class=MaterialExpressionMultiply
   inputs [
    A {from=MaterialExpressionTime_3}
    B {from=MaterialExpressionScalarParameter_31}
   ]
  }
  MaterialExpressionMultiply_42 {
   class=MaterialExpressionMultiply
   inputs [
    A {from=MaterialExpressionTime_9}
    B {from=MaterialExpressionScalarParameter_31}
   ]
  }
  MaterialExpressionMultiply_43 {
   class=MaterialExpressionMultiply
   inputs [
    A {from=MaterialExpressionSaturate_3}
    B {from=MaterialExpressionScalarParameter_40}
   ]
  }
  MaterialExpressionMultiply_44 {
   class=MaterialExpressionMultiply
   inputs [
    A {from=MaterialExpressionTextureSample_1, output=R}
    B {from=MaterialExpressionScalarParameter_44}
   ]
  }
  MaterialExpressionMultiply_45 {
   class=MaterialExpressionMultiply
   inputs [
    A {from=MaterialExpressionAdd_5}
    B {from=MaterialExpressionScalarParameter_42}
   ]
  }
  MaterialExpressionMultiply_46 {
   class=MaterialExpressionMultiply
   inputs [
    A {from=MaterialExpressionMultiply_32}
    B {from=MaterialExpressionNamedRerouteUsage_11}
   ]
  }
  MaterialExpressionMultiply_47 {
   class=MaterialExpressionMultiply
   inputs [
    A {from=MaterialExpressionMaterialFunctionCall_20, output=Result}
    B {from=MaterialExpressionMaterialFunctionCall_23, output=Result}
   ]
  }
  MaterialExpressionMultiply_49 {
   class=MaterialExpressionMultiply
   inputs [
    A {from=MaterialExpressionMaterialFunctionCall_40, output=StabilizedUVs}
    B {from=""}
   ]
   properties {constB=0.5}
  }
  MaterialExpressionMultiply_5 {
   class=MaterialExpressionMultiply
   inputs [
    A {from=MaterialExpressionMultiply_12}
    B {from=MaterialExpressionMultiply_43}
   ]
  }
  MaterialExpressionMultiply_50 {
   class=MaterialExpressionMultiply
   inputs [
    A {from=MaterialExpressionNamedRerouteDeclaration_8}
    B {from=MaterialExpressionNamedRerouteUsage_2}
   ]
  }
  MaterialExpressionMultiply_52 {
   class=MaterialExpressionMultiply
   inputs [
    A {from=MaterialExpressionFresnel_1}
    B {from=MaterialExpressionScalarParameter_59}
   ]
  }
  MaterialExpressionMultiply_53 {
   class=MaterialExpressionMultiply
   inputs [
    A {from=MaterialExpressionTextureCoordinate_3}
    B {from=MaterialExpressionScalarParameter_50}
   ]
  }
  MaterialExpressionMultiply_55 {
   class=MaterialExpressionMultiply
   inputs [
    A {from=MaterialExpressionScalarParameter_52}
    B {from=MaterialExpressionStep_4}
   ]
  }
  MaterialExpressionMultiply_56 {
   class=MaterialExpressionMultiply
   inputs [
    A {from=MaterialExpressionSubtract_2}
    B {from=MaterialExpressionScalarParameter_51}
   ]
  }
  MaterialExpressionMultiply_59 {
   class=MaterialExpressionMultiply
   inputs [
    A {from=MaterialExpressionTime_5}
    B {from=MaterialExpressionMaterialFunctionCall_26, output=Result}
   ]
  }
  MaterialExpressionMultiply_6 {
   class=MaterialExpressionMultiply
   inputs [
    A {from=MaterialExpressionMultiply_4}
    B {from=MaterialExpressionMax_2}
   ]
  }
  MaterialExpressionMultiply_60 {
   class=MaterialExpressionMultiply
   inputs [
    A {from=MaterialExpressionScalarParameter_10}
    B {from=MaterialExpressionAdd_19}
   ]
  }
  MaterialExpressionMultiply_61 {
   class=MaterialExpressionMultiply
   inputs [
    A {from=MaterialExpressionMax_6}
    B {from=MaterialExpressionPower_0}
   ]
  }
  MaterialExpressionMultiply_7 {
   class=MaterialExpressionMultiply
   inputs [
    A {from=MaterialExpressionTextureSample_4, output=R}
    B {from=MaterialExpressionScalarParameter_19}
   ]
  }
  MaterialExpressionMultiply_8 {
   class=MaterialExpressionMultiply
   inputs [
    A {from=MaterialExpressionTextureCoordinate_0}
    B {from=MaterialExpressionMaterialFunctionCall_14, output=Result}
   ]
  }
  MaterialExpressionNamedRerouteDeclaration_0 {
   class=MaterialExpressionNamedRerouteDeclaration
   inputs [None {from=MaterialExpressionGetMaterialAttributes_0, output=Normal}]
   properties {name=Normals, nodeColor=<rgba 0.2656 1 0 1>}
  }
  MaterialExpressionNamedRerouteDeclaration_1 {
   class=MaterialExpressionNamedRerouteDeclaration
   inputs [None {from=MaterialExpressionDivide_3}]
   properties {name=Cube Tiling, nodeColor=<rgba 0 1 0.625 1>}
  }
  MaterialExpressionNamedRerouteDeclaration_10 {
   class=MaterialExpressionNamedRerouteDeclaration
   inputs [None {from=MaterialExpressionMultiply_11}]
   properties {name=Glitch Mask, nodeColor=<rgba 0.4531 1 0 1>}
  }
  MaterialExpressionNamedRerouteDeclaration_11 {
   class=MaterialExpressionNamedRerouteDeclaration
   inputs [None {from=MaterialExpressionLinearInterpolate_12}]
   properties {name=UV1/2, nodeColor=<rgba 0.4531 0 1 1>}
  }
  MaterialExpressionNamedRerouteDeclaration_12 {
   class=MaterialExpressionNamedRerouteDeclaration
   inputs [None {from=MaterialExpressionMaterialFunctionCall_17, output=Parallax UVs}]
   properties {name=Camo UV, nodeColor=<rgba 1 0 0.7266 1>}
  }
  MaterialExpressionNamedRerouteDeclaration_13 {
   class=MaterialExpressionNamedRerouteDeclaration
   inputs [None {from=MaterialExpressionMaterialFunctionCall_29, output=Result}]
   properties {name=Dot Animation- Mask, nodeColor=<rgba 0.3125 0 1 1>}
  }
  MaterialExpressionNamedRerouteDeclaration_14 {
   class=MaterialExpressionNamedRerouteDeclaration
   inputs [None {from=MaterialExpressionGetMaterialAttributes_0, output=Roughness}]
   properties {name=Roughness In, nodeColor=<rgba 1 0.1875 0 1>}
  }
  MaterialExpressionNamedRerouteDeclaration_15 {
   class=MaterialExpressionNamedRerouteDeclaration
   inputs [None {from=MaterialExpressionMax_5}]
   properties {name=Hood, nodeColor=<rgba 1 0 0.6094 1>}
  }
  MaterialExpressionNamedRerouteDeclaration_16 {
   class=MaterialExpressionNamedRerouteDeclaration
   inputs [None {from=MaterialExpressionSaturate_0}]
   properties {name=Camo Area, nodeColor=<rgba 0.4766 0 1 1>}
  }
  MaterialExpressionNamedRerouteDeclaration_17 {
   class=MaterialExpressionNamedRerouteDeclaration
   inputs [None {from=MaterialExpressionMultiply_18}]
   properties {name=Dots, nodeColor=<rgba 0.0078 1 0 1>}
  }
  MaterialExpressionNamedRerouteDeclaration_18 {
   class=MaterialExpressionNamedRerouteDeclaration
   inputs [None {from=MaterialExpressionMultiply_25}]
   properties {name=Scanline, nodeColor=<rgba 0 1 0.2266 1>}
  }
  MaterialExpressionNamedRerouteDeclaration_19 {
   class=MaterialExpressionNamedRerouteDeclaration
   inputs [None {from=MaterialExpressionMultiply_27}]
   properties {name=Camo, nodeColor=<rgba 1 0 0.0938 1>}
  }
  MaterialExpressionNamedRerouteDeclaration_20 {
   class=MaterialExpressionNamedRerouteDeclaration
   inputs [None {from=MaterialExpressionMultiply_52}]
   properties {name=Inner Fresnel, nodeColor=<rgba 0 0.4141 1 1>}
  }
  MaterialExpressionNamedRerouteDeclaration_21 {
   class=MaterialExpressionNamedRerouteDeclaration
   inputs [None {from=MaterialExpressionAdd_4}]
   properties {name=Outer Fresnel, nodeColor=<rgba 0.7344 1 0 1>}
  }
  MaterialExpressionNamedRerouteDeclaration_22 {
   class=MaterialExpressionNamedRerouteDeclaration
   inputs [None {from=MaterialExpressionLinearInterpolate_2}]
   properties {name=Roughness Out, nodeColor=<rgba 0 1 0.25 1>}
  }
  MaterialExpressionNamedRerouteDeclaration_23 {
   class=MaterialExpressionNamedRerouteDeclaration
   inputs [None {from=MaterialExpressionLinearInterpolate_4}]
   properties {name=Base Colour Out, nodeColor=<rgba 1 0 0.0703 1>}
  }
  MaterialExpressionNamedRerouteDeclaration_24 {
   class=MaterialExpressionNamedRerouteDeclaration
   inputs [None {from=MaterialExpressionFunctionInput_0}]
   properties {name=Tech Art Mask, nodeColor=<rgba 1 0.3047 0 1>}
  }
  MaterialExpressionNamedRerouteDeclaration_26 {
   class=MaterialExpressionNamedRerouteDeclaration
   inputs [None {from=MaterialExpressionStaticComponentMaskParameter_0}]
   properties {name=Camo Mask, nodeColor=<rgba 0 0.9531 1 1>}
  }
  MaterialExpressionNamedRerouteDeclaration_27 {
   class=MaterialExpressionNamedRerouteDeclaration
   inputs [None {from=MaterialExpressionLinearInterpolate_16}]
   properties {name=Emissive, nodeColor=<rgba 1 0.7266 0 1>}
  }
  MaterialExpressionNamedRerouteDeclaration_28 {
   class=MaterialExpressionNamedRerouteDeclaration
   inputs [None {from=MaterialExpressionGetMaterialAttributes_0, output=EmissiveColor}]
   properties {name=Emissive In, nodeColor=<rgba 0.4062 0 1 1>}
  }
  MaterialExpressionNamedRerouteDeclaration_29 {
   class=MaterialExpressionNamedRerouteDeclaration
   inputs [None {from=MaterialExpressionGetMaterialAttributes_0, output=Specular}]
   properties {name=Spec In, nodeColor=<rgba 0 1 0.0859 1>}
  }
  MaterialExpressionNamedRerouteDeclaration_3 {
   class=MaterialExpressionNamedRerouteDeclaration
   inputs [None {from=MaterialExpressionSaturate_1}]
   properties {name=Emissive Mask, nodeColor=<rgba 0.1719 0 1 1>}
  }
  MaterialExpressionNamedRerouteDeclaration_30 {
   class=MaterialExpressionNamedRerouteDeclaration
   inputs [None {from=MaterialExpressionNamedRerouteUsage_34}]
   properties {name=Spec Out, nodeColor=<rgba 1 0 0.2344 1>}
  }
  MaterialExpressionNamedRerouteDeclaration_31 {
   class=MaterialExpressionNamedRerouteDeclaration
   inputs [None {from=MaterialExpressionLinearInterpolate_0}]
   properties {name=Emissive Out, nodeColor=<rgba 0 0.5547 1 1>}
  }
  MaterialExpressionNamedRerouteDeclaration_4 {
   class=MaterialExpressionNamedRerouteDeclaration
   inputs [None {from=MaterialExpressionLinearInterpolate_15}]
   properties {name=Flicker, nodeColor=<rgba 0 1 1 1>}
  }
  MaterialExpressionNamedRerouteDeclaration_5 {
   class=MaterialExpressionNamedRerouteDeclaration
   inputs [None {from=MaterialExpressionTextureSampleParameter2D_0, output=G}]
   properties {name=Legs/Hood, nodeColor=<rgba 0.1484 1 0 1>}
  }
  MaterialExpressionNamedRerouteDeclaration_6 {
   class=MaterialExpressionNamedRerouteDeclaration
   inputs [None {from=MaterialExpressionGetMaterialAttributes_0, output=BaseColor}]
   properties {name=Base Color In, nodeColor=<rgba 0.5703 0 1 1>}
  }
  MaterialExpressionNamedRerouteDeclaration_7 {
   class=MaterialExpressionNamedRerouteDeclaration
   inputs [None {from=MaterialExpressionDivide_1}]
   properties {name=Time Stepping, nodeColor=<rgba 0.2891 0 1 1>}
  }
  MaterialExpressionNamedRerouteDeclaration_8 {
   class=MaterialExpressionNamedRerouteDeclaration
   inputs [None {from=MaterialExpressionStaticSwitchParameter_2}]
   properties {name=Grid Dot UVs, nodeColor=<rgba 1 0.7734 0 1>}
  }
  MaterialExpressionNamedRerouteDeclaration_9 {
   class=MaterialExpressionNamedRerouteDeclaration
   inputs [None {from=MaterialExpressionAdd_1}]
   properties {name=Glitches UV, nodeColor=<rgba 0.4297 0 1 1>}
  }
  MaterialExpressionNamedRerouteUsage_1 {class=MaterialExpressionNamedRerouteUsage}
  MaterialExpressionNamedRerouteUsage_10 {class=MaterialExpressionNamedRerouteUsage}
  MaterialExpressionNamedRerouteUsage_11 {class=MaterialExpressionNamedRerouteUsage}
  MaterialExpressionNamedRerouteUsage_12 {class=MaterialExpressionNamedRerouteUsage}
  MaterialExpressionNamedRerouteUsage_13 {class=MaterialExpressionNamedRerouteUsage}
  MaterialExpressionNamedRerouteUsage_14 {class=MaterialExpressionNamedRerouteUsage}
  MaterialExpressionNamedRerouteUsage_15 {class=MaterialExpressionNamedRerouteUsage}
  MaterialExpressionNamedRerouteUsage_16 {class=MaterialExpressionNamedRerouteUsage}
  MaterialExpressionNamedRerouteUsage_17 {class=MaterialExpressionNamedRerouteUsage}
  MaterialExpressionNamedRerouteUsage_18 {class=MaterialExpressionNamedRerouteUsage}
  MaterialExpressionNamedRerouteUsage_19 {class=MaterialExpressionNamedRerouteUsage}
  MaterialExpressionNamedRerouteUsage_2 {class=MaterialExpressionNamedRerouteUsage}
  MaterialExpressionNamedRerouteUsage_20 {class=MaterialExpressionNamedRerouteUsage}
  MaterialExpressionNamedRerouteUsage_21 {class=MaterialExpressionNamedRerouteUsage}
  MaterialExpressionNamedRerouteUsage_22 {class=MaterialExpressionNamedRerouteUsage}
  MaterialExpressionNamedRerouteUsage_28 {class=MaterialExpressionNamedRerouteUsage}
  MaterialExpressionNamedRerouteUsage_29 {class=MaterialExpressionNamedRerouteUsage}
  MaterialExpressionNamedRerouteUsage_32 {class=MaterialExpressionNamedRerouteUsage}
  MaterialExpressionNamedRerouteUsage_33 {class=MaterialExpressionNamedRerouteUsage}
  MaterialExpressionNamedRerouteUsage_34 {class=MaterialExpressionNamedRerouteUsage}
  MaterialExpressionNamedRerouteUsage_35 {class=MaterialExpressionNamedRerouteUsage}
  MaterialExpressionNamedRerouteUsage_36 {class=MaterialExpressionNamedRerouteUsage}
  MaterialExpressionNamedRerouteUsage_5 {class=MaterialExpressionNamedRerouteUsage}
  MaterialExpressionNamedRerouteUsage_6 {class=MaterialExpressionNamedRerouteUsage}
  MaterialExpressionNamedRerouteUsage_7 {class=MaterialExpressionNamedRerouteUsage}
  MaterialExpressionNamedRerouteUsage_8 {class=MaterialExpressionNamedRerouteUsage}
  MaterialExpressionNamedRerouteUsage_9 {class=MaterialExpressionNamedRerouteUsage}
  MaterialExpressionOneMinus_0 {class=MaterialExpressionOneMinus, inputs [None {from=MaterialExpressionFrac_2}]}
  MaterialExpressionOneMinus_1 {class=MaterialExpressionOneMinus, inputs [None {from=MaterialExpressionFresnel_0}]}
  MaterialExpressionOneMinus_2 {class=MaterialExpressionOneMinus, inputs [None {from=MaterialExpressionMaterialFunctionCall_41, output=Result}]}
  MaterialExpressionOneMinus_4 {class=MaterialExpressionOneMinus, inputs [None {from=MaterialExpressionTextureSampleParameter2D_0, output=B}]}
  MaterialExpressionPanner_0 {
   class=MaterialExpressionPanner
   inputs [
    Coordinate {from=MaterialExpressionMaterialFunctionCall_8, output=Rotated Values}
    Time {from=MaterialExpressionNamedRerouteDeclaration_7}
    Speed {from=MaterialExpressionMaterialFunctionCall_5, output=Result}
   ]
   properties {speedY=0.1}
  }
  MaterialExpressionPower_0 {
   class=MaterialExpressionPower
   inputs [
    Base {from=MaterialExpressionOneMinus_2}
    Exp {from=""}
   ]
   properties {constExponent=1.8}
  }
  MaterialExpressionRotator_1 {
   class=MaterialExpressionRotator
   inputs [
    Coordinate {from=MaterialExpressionMultiply_49, output=""}
    Time {from=MaterialExpressionMaterialFunctionCall_6, output=Result}
   ]
   properties {}
  }
  MaterialExpressionSaturate_0 {class=MaterialExpressionSaturate, inputs [None {from=MaterialExpressionSubtract_0}]}
  MaterialExpressionSaturate_1 {class=MaterialExpressionSaturate, inputs [None {from=MaterialExpressionTextureSampleParameter2D_0, output=R}]}
  MaterialExpressionSaturate_2 {class=MaterialExpressionSaturate, inputs [None {from=MaterialExpressionFrac_0}]}
  MaterialExpressionSaturate_3 {class=MaterialExpressionSaturate, inputs [None {from=MaterialExpressionMaterialFunctionCall_15, output=Result}]}
  MaterialExpressionSaturate_4 {class=MaterialExpressionSaturate, inputs [None {from=MaterialExpressionSubtract_1}]}
  MaterialExpressionScalarParameter_0 {class=MaterialExpressionScalarParameter, properties {defaultValue=0.58, parameterName=Outer Fresnel - Exp}}
  MaterialExpressionScalarParameter_1 {class=MaterialExpressionScalarParameter, properties {defaultValue=0.1, parameterName="Outer Fresnel -  Bias"}}
  MaterialExpressionScalarParameter_10 {class=MaterialExpressionScalarParameter, properties {defaultValue=12.5, group=Paw Jasmine, parameterName=Camo - Distort Tiling}}
  MaterialExpressionScalarParameter_11 {class=MaterialExpressionScalarParameter, properties {defaultValue=0.01, group=Paw Jasmine, parameterName=Camo - Distort Intensity}}
  MaterialExpressionScalarParameter_12 {class=MaterialExpressionScalarParameter, properties {defaultValue=0.2, parameterName=Inner Fresnel - Step}}
  MaterialExpressionScalarParameter_13 {class=MaterialExpressionScalarParameter, properties {defaultValue=0.552, parameterName=Inner Fresnel - Smoothness}}
  MaterialExpressionScalarParameter_14 {class=MaterialExpressionScalarParameter, properties {defaultValue=0.9, group=Paw Jasmine, parameterName=Camo - Y - Tiling}}
  MaterialExpressionScalarParameter_15 {class=MaterialExpressionScalarParameter, properties {defaultValue=0.5, group=Paw Jasmine, parameterName=Circuit Dots - Texture Brightness}}
  MaterialExpressionScalarParameter_16 {class=MaterialExpressionScalarParameter, properties {defaultValue=0.125, group=Paw Jasmine, parameterName=Dots - Texture Tiling}}
  MaterialExpressionScalarParameter_17 {class=MaterialExpressionScalarParameter, properties {defaultValue=32, group=Paw Jasmine, parameterName=Steps}}
  MaterialExpressionScalarParameter_18 {class=MaterialExpressionScalarParameter, properties {defaultValue=150, group=Paw Jasmine, parameterName=Scanline - Tiling}}
  MaterialExpressionScalarParameter_19 {class=MaterialExpressionScalarParameter, properties {defaultValue=2, group=Paw Jasmine, parameterName=Glitch Distortion - Intensity}}
  MaterialExpressionScalarParameter_2 {class=MaterialExpressionScalarParameter, properties {parameterName=Glitch Block - Steps}}
  MaterialExpressionScalarParameter_20 {class=MaterialExpressionScalarParameter, properties {defaultValue=0.6739, group=Paw Jasmine, parameterName=Scanline - Length}}
  MaterialExpressionScalarParameter_21 {class=MaterialExpressionScalarParameter, properties {defaultValue=0.552, group=Paw Jasmine, parameterName=Scanline - Bottom Intensity}}
  MaterialExpressionScalarParameter_22 {class=MaterialExpressionScalarParameter, properties {defaultValue=0.5, group=Paw Jasmine, parameterName=Glow - Flicker Intensity}}
  MaterialExpressionScalarParameter_23 {class=MaterialExpressionScalarParameter, properties {group=Paw Jasmine, parameterName=Enable Reactivity?}}
  MaterialExpressionScalarParameter_24 {class=MaterialExpressionScalarParameter, properties {defaultValue=0.99, group=Paw Jasmine, parameterName=Glitch - Step}}
  MaterialExpressionScalarParameter_25 {class=MaterialExpressionScalarParameter, properties {defaultValue=0.1, group=Paw Jasmine, parameterName=Scanline - Intensity}}
  MaterialExpressionScalarParameter_26 {class=MaterialExpressionScalarParameter, properties {defaultValue=0.01, group=Paw Jasmine, parameterName=Scanline - Speed}}
  MaterialExpressionScalarParameter_28 {class=MaterialExpressionScalarParameter, properties {defaultValue=200, group=Paw Jasmine, parameterName="Dot - Amount  (Tiling)"}}
  MaterialExpressionScalarParameter_29 {class=MaterialExpressionScalarParameter, properties {defaultValue=0.47, group=Paw Jasmine, parameterName=Cloud Dots - Step Amount}}
  MaterialExpressionScalarParameter_3 {class=MaterialExpressionScalarParameter, properties {group=Paw Jasmine, parameterName=WPO Offset}}
  MaterialExpressionScalarParameter_30 {class=MaterialExpressionScalarParameter, properties {defaultValue=3.5, group=Paw Jasmine, parameterName=Dot Clouds - Mask Tiling}}
  MaterialExpressionScalarParameter_31 {class=MaterialExpressionScalarParameter, properties {defaultValue=2, group=Paw Jasmine, parameterName=Glow - Flicker Speed}}
  MaterialExpressionScalarParameter_32 {class=MaterialExpressionScalarParameter, properties {defaultValue=0.02, group=Paw Jasmine, parameterName=Dot Clouds - Mask Speed}}
  MaterialExpressionScalarParameter_33 {class=MaterialExpressionScalarParameter, properties {defaultValue=0.7623, group=Paw Jasmine, parameterName=Dots - Moire Mitigation}}
  MaterialExpressionScalarParameter_34 {class=MaterialExpressionScalarParameter, properties {defaultValue=0.5, group=Paw Jasmine, parameterName=Cloud Dots - Texture Brightness}}
  MaterialExpressionScalarParameter_36 {class=MaterialExpressionScalarParameter, properties {defaultValue=0.25, group=Paw Jasmine, parameterName=UV - Rotation}}
  MaterialExpressionScalarParameter_37 {class=MaterialExpressionScalarParameter, properties {defaultValue=2, group=Paw Jasmine, parameterName=Glitch X}}
  MaterialExpressionScalarParameter_38 {class=MaterialExpressionScalarParameter, properties {defaultValue=2, group=Paw Jasmine, parameterName=Glitch Y}}
  MaterialExpressionScalarParameter_39 {class=MaterialExpressionScalarParameter, properties {defaultValue=5, group=Paw Jasmine, parameterName=Glitch Distortion - Y - Tiling}}
  MaterialExpressionScalarParameter_4 {class=MaterialExpressionScalarParameter, properties {defaultValue=-0.1, group=Paw Jasmine, parameterName=Glitch - Speed Y}}
  MaterialExpressionScalarParameter_40 {class=MaterialExpressionScalarParameter, properties {defaultValue=1, group=Paw Jasmine, parameterName=Glitch Opacity}}
  MaterialExpressionScalarParameter_41 {class=MaterialExpressionScalarParameter, properties {defaultValue=1, group=Paw Jasmine, parameterName=Glitch - Emissive Brightness}}
  MaterialExpressionScalarParameter_42 {class=MaterialExpressionScalarParameter, properties {defaultValue=0.2, group=Paw Jasmine, parameterName=Dot - Animation Speed}}
  MaterialExpressionScalarParameter_43 {class=MaterialExpressionScalarParameter, properties {defaultValue=0.5, group=Paw Jasmine, parameterName=Circuit Dots - Intensity}}
  MaterialExpressionScalarParameter_44 {class=MaterialExpressionScalarParameter, properties {defaultValue=1, group=Paw Jasmine, parameterName=Glitch - Compression Intensity}}
  MaterialExpressionScalarParameter_45 {class=MaterialExpressionScalarParameter, properties {defaultValue=0.3, group=Paw Jasmine, parameterName=Glitch Offset}}
  MaterialExpressionScalarParameter_46 {class=MaterialExpressionScalarParameter, properties {defaultValue=1, group=Paw Jasmine, parameterName=Use UV2?}}
  MaterialExpressionScalarParameter_47 {class=MaterialExpressionScalarParameter, properties {defaultValue=5, group=Paw Jasmine, parameterName="Dots -  Modulation Strength"}}
  MaterialExpressionScalarParameter_5 {class=MaterialExpressionScalarParameter, properties {group=Paw Jasmine, parameterName=Camo - Speed Y}}
  MaterialExpressionScalarParameter_50 {class=MaterialExpressionScalarParameter, properties {defaultValue=3, group=Paw Jasmine, parameterName=Dots - Modulation Tiling}}
  MaterialExpressionScalarParameter_51 {class=MaterialExpressionScalarParameter, properties {defaultValue=1, group=Paw Jasmine, parameterName=Hood Off}}
  MaterialExpressionScalarParameter_52 {class=MaterialExpressionScalarParameter, properties {defaultValue=1, group=Paw Jasmine, parameterName=Hood On}}
  MaterialExpressionScalarParameter_54 {class=MaterialExpressionScalarParameter, properties {defaultValue=0.3, group=Paw Jasmine, parameterName=Dots - Position Select}}
  MaterialExpressionScalarParameter_55 {class=MaterialExpressionScalarParameter, properties {defaultValue=0.1654, group=Paw Jasmine, parameterName=Dots - Range}}
  MaterialExpressionScalarParameter_56 {class=MaterialExpressionScalarParameter, properties {group=Paw Jasmine, parameterName=Hood Reactivity}}
  MaterialExpressionScalarParameter_57 {class=MaterialExpressionScalarParameter, properties {defaultValue=5, group=Paw Jasmine, parameterName=Dot - Brightness}}
  MaterialExpressionScalarParameter_58 {class=MaterialExpressionScalarParameter, properties {defaultValue=3.3027, parameterName=Outer Fresnel 2 - Exp}}
  MaterialExpressionScalarParameter_59 {class=MaterialExpressionScalarParameter, properties {defaultValue=1, parameterName=Outer Fresnel 2 - Intensity}}
  MaterialExpressionScalarParameter_6 {class=MaterialExpressionScalarParameter, properties {defaultValue=-0.2, group=Paw Jasmine, parameterName=Camo - Bump Offset - Depth}}
  MaterialExpressionScalarParameter_60 {class=MaterialExpressionScalarParameter, properties {defaultValue=1, parameterName=Emissive Camo}}
  MaterialExpressionScalarParameter_61 {class=MaterialExpressionScalarParameter, properties {defaultValue=0.3101, group=Paw Jasmine, parameterName=Clear Coat - Under Roughness}}
  MaterialExpressionScalarParameter_67 {class=MaterialExpressionScalarParameter, properties {defaultValue=200, group=Paw Jasmine, parameterName=Dot - Opacity - Fade Offset}}
  MaterialExpressionScalarParameter_68 {class=MaterialExpressionScalarParameter, properties {defaultValue=1100, group=Paw Jasmine, parameterName=Dot - Opacity - Fade Length}}
  MaterialExpressionScalarParameter_7 {class=MaterialExpressionScalarParameter, properties {defaultValue=0.9, group=Paw Jasmine, parameterName=Camo - X - Tiling}}
  MaterialExpressionScalarParameter_76 {class=MaterialExpressionScalarParameter, properties {group=Paw Jasmine, parameterName=Disable Reactivity CID}}
  MaterialExpressionScalarParameter_8 {class=MaterialExpressionScalarParameter, properties {defaultValue=0.5, parameterName=Inner Fresnel - Exp}}
  MaterialExpressionScalarParameter_9 {class=MaterialExpressionScalarParameter, properties {defaultValue=-0.01, group=Paw Jasmine, parameterName=Camo - Distort Speed}}
  MaterialExpressionSetMaterialAttributes_0 {
   class=MaterialExpressionSetMaterialAttributes
   inputs [
    MaterialAttributes {from=MaterialExpressionGetMaterialAttributes_0, output=MaterialAttributes}
    EmissiveColor {from=MaterialExpressionNamedRerouteUsage_36, output=""}
    Specular {from=MaterialExpressionNamedRerouteUsage_35, output=""}
    Roughness {from=MaterialExpressionNamedRerouteUsage_16, output=""}
    BaseColor {from=MaterialExpressionNamedRerouteUsage_17, output=""}
   ]
   properties {attributeSetTypes [
     #1
     #2
     #3
     #4
    ]}
  }
  MaterialExpressionSetMaterialAttributes_1 {
   class=MaterialExpressionSetMaterialAttributes
   inputs [
    MaterialAttributes {from=MaterialExpressionGetMaterialAttributes_2, output=MaterialAttributes}
    WorldPositionOffset {from=MaterialExpressionLinearInterpolate_8, output=""}
    Opacity {from=MaterialExpressionMultiply_5, output=""}
    EmissiveColor {from=MaterialExpressionMultiply_6, output=""}
    OpacityMask {from=MaterialExpressionMaterialFunctionCall_16, output=Result}
   ]
   properties {attributeSetTypes [
     #6
     #7
     #1
     #8
    ]}
  }
  MaterialExpressionSine_0 {class=MaterialExpressionSine, inputs [None {from=MaterialExpressionMultiply_41}]}
  MaterialExpressionSine_1 {class=MaterialExpressionSine, inputs [None {from=MaterialExpressionMultiply_39}]}
  MaterialExpressionSine_2 {class=MaterialExpressionSine, inputs [None {from=MaterialExpressionMultiply_40}]}
  MaterialExpressionSine_4 {class=MaterialExpressionSine, inputs [None {from=MaterialExpressionMultiply_42}]}
  MaterialExpressionSmoothStep_0 {
   class=MaterialExpressionSmoothStep
   inputs [
    Min {from=MaterialExpressionScalarParameter_12, output=""}
    Max {from=MaterialExpressionAdd_3, output=""}
    Value {from=MaterialExpressionFresnel_2, output=""}
   ]
   properties {}
  }
  MaterialExpressionSmoothStep_3 {
   class=MaterialExpressionSmoothStep
   inputs [
    Min {from=MaterialExpressionScalarParameter_20, output=""}
    Max {from="", output=""}
    Value {from=MaterialExpressionFrac_2, output=""}
   ]
   properties {}
  }
  MaterialExpressionSmoothStep_4 {
   class=MaterialExpressionSmoothStep
   inputs [
    Min {from=MaterialExpressionScalarParameter_20, output=""}
    Max {from="", output=""}
    Value {from=MaterialExpressionOneMinus_0, output=""}
   ]
   properties {}
  }
  MaterialExpressionStaticBoolParameter_0 {class=MaterialExpressionStaticBoolParameter, inputs [], properties {parameterName="Apply Offset (SB)"}}
  MaterialExpressionStaticComponentMaskParameter_0 {
   class=MaterialExpressionStaticComponentMaskParameter
   inputs [None {from=MaterialExpressionNamedRerouteUsage_28, output=""}]
   properties {defaultR=true, parameterName=Camo Mask}
  }
  MaterialExpressionStaticSwitchParameter_0 {
   class=MaterialExpressionStaticSwitchParameter
   inputs [
    True {from=MaterialExpressionSetMaterialAttributes_0, output=""}
    False {from=MaterialExpressionFunctionInput_1, output=""}
   ]
   properties {defaultValue=true, group=Paw Jasmine - Switches, parameterName=Use Camo Effect?}
  }
  MaterialExpressionStaticSwitchParameter_1 {
   class=MaterialExpressionStaticSwitchParameter
   inputs [
    True {from=MaterialExpressionSetMaterialAttributes_1, output=""}
    False {from=MaterialExpressionStaticSwitchParameter_0, output=""}
   ]
   properties {group=Paw Jasmine - Switches, parameterName=Overlay Material?}
  }
  MaterialExpressionStaticSwitchParameter_2 {
   class=MaterialExpressionStaticSwitchParameter
   inputs [
    True {from=MaterialExpressionRotator_1, output=""}
    False {from=MaterialExpressionNamedRerouteUsage_10, output=""}
   ]
   properties {defaultValue=true, parameterName=Use Screenspace UVs?}
  }
  MaterialExpressionStep_0 {
   class=MaterialExpressionStep
   inputs [
    Y {from=MaterialExpressionOneMinus_4, output=""}
    X {from=MaterialExpressionScalarParameter_24, output=""}
   ]
   properties {constY=0.01}
  }
  MaterialExpressionStep_1 {
   class=MaterialExpressionStep
   inputs [
    Y {from=MaterialExpressionTextureSample_13, output=RGB}
    X {from=MaterialExpressionScalarParameter_29, output=""}
   ]
   properties {}
  }
  MaterialExpressionStep_2 {
   class=MaterialExpressionStep
   inputs [
    Y {from=MaterialExpressionTextureSampleParameter2D_0, output=B}
    X {from=MaterialExpressionScalarParameter_24, output=""}
   ]
   properties {constY=0.01}
  }
  MaterialExpressionStep_4 {
   class=MaterialExpressionStep
   inputs [
    Y {from="", output=""}
    X {from=MaterialExpressionNamedRerouteDeclaration_5, output=""}
   ]
   properties {constX=0.01, constY=0.9}
  }
  MaterialExpressionSubtract_0 {
   class=MaterialExpressionSubtract
   inputs [
    A {from=MaterialExpressionNamedRerouteDeclaration_26}
    B {from=MaterialExpressionDesaturation_0}
   ]
  }
  MaterialExpressionSubtract_1 {
   class=MaterialExpressionSubtract
   inputs [
    A {from=MaterialExpressionSmoothStep_0}
    B {from=MaterialExpressionNamedRerouteUsage_21}
   ]
  }
  MaterialExpressionSubtract_2 {
   class=MaterialExpressionSubtract
   inputs [
    A {from=MaterialExpressionNamedRerouteDeclaration_5}
    B {from=MaterialExpressionStep_4}
   ]
  }
  MaterialExpressionSubtract_3 {
   class=MaterialExpressionSubtract
   inputs [
    A {from=MaterialExpressionTextureSample_1, output=G}
    B {from=MaterialExpressionScalarParameter_2}
   ]
  }
  MaterialExpressionTextureCoordinate_0 {class=MaterialExpressionTextureCoordinate, properties {coordinateIndex=1}}
  MaterialExpressionTextureCoordinate_1 {class=MaterialExpressionTextureCoordinate}
  MaterialExpressionTextureCoordinate_3 {class=MaterialExpressionTextureCoordinate}
  MaterialExpressionTextureCoordinate_4 {class=MaterialExpressionTextureCoordinate, properties {coordinateIndex=1}}
  MaterialExpressionTextureObject_0 {class=MaterialExpressionTextureObject, inputs [], properties {texture=/Game/ContentCreationTools/ObjectPlacement/Textures/T_BlankWhite_01}}
  MaterialExpressionTextureSampleParameter2D_0 {
   class=MaterialExpressionTextureSampleParameter2D
   inputs [
    UVs {from="", output=""}
    Apply View MipBias {from="", output=""}
   ]
   properties {
    group=Paw Jasmine
    parameterName=FX2
    samplerType=SAMPLERTYPE_Masks
    texture=/BRCosmetics/Characters/Player/Male/Medium/Bodies/M_MED_PawJasmineGravity/Skins/Greet/Textures/T_PawJasmineGravity_Greet_Body_FX2
   }
  }
  MaterialExpressionTextureSampleParameter2D_1 {
   class=MaterialExpressionTextureSampleParameter2D
   inputs [
    UVs {from=MaterialExpressionNamedRerouteDeclaration_12, output=""}
    Apply View MipBias {from="", output=""}
   ]
   properties {
    parameterName=Camo Texture
    texture=/BRCosmetics/Characters/Player/Male/Medium/Bodies/M_MED_PawJasmineGravity/Skins/Greet/Textures/T_PawJasmineGravity_Greet_Tilable_D
   }
  }
  MaterialExpressionTextureSample_0 {
   class=MaterialExpressionTextureSample
   inputs [
    UVs {from=MaterialExpressionMultiply_31}
    Tex {from=""}
    Apply View MipBias {from=""}
   ]
   properties {texture=/BRCosmetics/Characters/Player/Female/Medium/Bodies/F_MED_PawJasmineSword/Skins/Mane/Textures/T_Dot_M}
  }
  MaterialExpressionTextureSample_1 {
   class=MaterialExpressionTextureSample
   inputs [
    UVs {from=MaterialExpressionAdd_2}
    Tex {from=""}
    Apply View MipBias {from=""}
   ]
   properties {
    samplerType=SAMPLERTYPE_Masks
    texture=/BRCosmetics/Characters/Player/Male/Medium/Bodies/M_MED_PawJasmineGravity/Skins/Greet/Textures/T_Rectangle_Glitches_Tilable_01
   }
  }
  MaterialExpressionTextureSample_10 {
   class=MaterialExpressionTextureSample
   inputs [
    UVs {from=MaterialExpressionMultiply_16}
    Tex {from=""}
    Apply View MipBias {from=""}
   ]
   properties {
    samplerSource=SSM_Wrap_WorldGroupSettings
    samplerType=SAMPLERTYPE_LinearGrayscale
    texture=/BRCosmetics/Characters/Player/Male/Medium/Bodies/M_MED_PawJasmineGravity/Skins/Greet/Textures/LowResBlurredNoise_Linear
   }
  }
  MaterialExpressionTextureSample_12 {
   class=MaterialExpressionTextureSample
   inputs [
    UVs {from=MaterialExpressionMultiply_60}
    Tex {from=""}
    Apply View MipBias {from=""}
   ]
   properties {
    samplerSource=SSM_Wrap_WorldGroupSettings
    samplerType=SAMPLERTYPE_LinearGrayscale
    texture=&LowResBlurredNoise_Linear
   }
  }
  MaterialExpressionTextureSample_13 {
   class=MaterialExpressionTextureSample
   inputs [
    UVs {from=MaterialExpressionAdd_17}
    Tex {from=""}
    Apply View MipBias {from=""}
   ]
   properties {
    samplerSource=SSM_Wrap_WorldGroupSettings
    samplerType=SAMPLERTYPE_LinearGrayscale
    texture=&LowResBlurredNoise_Linear
   }
  }
  MaterialExpressionTextureSample_2 {
   class=MaterialExpressionTextureSample
   inputs [
    UVs {from=MaterialExpressionTextureSample_1, output=G}
    Tex {from=""}
    Apply View MipBias {from=""}
   ]
   properties {
    samplerType=SAMPLERTYPE_LinearColor
    texture=/BRCosmetics/Characters/Player/Male/Medium/Bodies/M_MED_PawJasmineGravity/Skins/Greet/Textures/CA_Glitches
   }
  }
  MaterialExpressionTextureSample_3 {
   class=MaterialExpressionTextureSample
   inputs [
    UVs {from=MaterialExpressionMultiply_50}
    Tex {from=""}
    Apply View MipBias {from=""}
   ]
   properties {
    samplerType=SAMPLERTYPE_LinearColor
    texture=/BRCosmetics/Characters/Player/Male/Medium/Bodies/M_MED_PawJasmineGravity/Skins/Greet/Textures/T_Overlapping_Cubes_Tilable_01
   }
  }
  MaterialExpressionTextureSample_4 {
   class=MaterialExpressionTextureSample
   inputs [
    UVs {from=MaterialExpressionMultiply_8}
    Tex {from=""}
    Apply View MipBias {from=""}
   ]
   properties {
    samplerType=SAMPLERTYPE_LinearColor
    texture=/BRCosmetics/Characters/Player/Male/Medium/Bodies/M_MED_ByteGear/Skins/Roll/Textures/T_ByteGear_Roll_Glitch_FX
   }
  }
  MaterialExpressionTextureSample_5 {
   class=MaterialExpressionTextureSample
   inputs [
    UVs {from=MaterialExpressionMultiply_53}
    Tex {from=""}
    Apply View MipBias {from=""}
   ]
   properties {samplerType=SAMPLERTYPE_Grayscale, texture=&LowResBlurredNoise}
  }
  MaterialExpressionTime_0 {class=MaterialExpressionTime, properties {bOverride_Period=true, period=0.4}}
  MaterialExpressionTime_1 {class=MaterialExpressionTime, properties {bOverride_Period=true, period=2.5}}
  MaterialExpressionTime_10 {class=MaterialExpressionTime}
  MaterialExpressionTime_11 {class=MaterialExpressionTime}
  MaterialExpressionTime_2 {class=MaterialExpressionTime}
  MaterialExpressionTime_3 {class=MaterialExpressionTime, properties {bOverride_Period=true, period=1.13}}
  MaterialExpressionTime_4 {class=MaterialExpressionTime}
  MaterialExpressionTime_5 {class=MaterialExpressionTime}
  MaterialExpressionTime_7 {class=MaterialExpressionTime}
  MaterialExpressionTime_8 {class=MaterialExpressionTime}
  MaterialExpressionTime_9 {class=MaterialExpressionTime, properties {bOverride_Period=true, period=1.5126}}
  MaterialExpressionTransform_1 {class=MaterialExpressionTransform, inputs [None {from=MaterialExpressionNamedRerouteUsage_1}]}
  MaterialExpressionVectorParameter_0 {class=MaterialExpressionVectorParameter, properties {parameterName=Inner Fresnel Color, defaultValue=<rgba 0.0156 0.0156 0.0156 0>}}
  MaterialExpressionVectorParameter_1 {class=MaterialExpressionVectorParameter, properties {group=Paw Jasmine, parameterName=Dot - Color, defaultValue=<rgba 0.5677 0.9185 1 1>}}
  MaterialExpressionVectorParameter_2 {class=MaterialExpressionVectorParameter, properties {group=Paw Jasmine, parameterName=Scanline Green, defaultValue=<rgba 0.3611 1 0.2497 1>}}
  MaterialExpressionVectorParameter_3 {class=MaterialExpressionVectorParameter, properties {group=Paw Jasmine, parameterName=Scanline Red, defaultValue=<rgba 1 0.0595 0.0499 1>}}
  MaterialExpressionVertexNormalWS_0 {class=MaterialExpressionVertexNormalWS, inputs [], properties {}}
  MaterialExpressionViewProperty_0 {class=MaterialExpressionViewProperty, inputs [], properties {property=MEVP_ResolutionFraction}}
 }}
