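# Clear Coat parent material for the Greet skin of the PawJasmineGravity character body, built on the shared 2023 character parent functions.
# PreFX output passes through optional, switch-gated hair and fly-away hair functions, then MF_PawJasmine and PostFX set the final material attributes.
# Side branches use MF_PawJasmine's Camo Area and Roughness reroutes, a bump-offset tiling normal and an outer roughness scalar for clear coat shading.
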Material /BRCosmetics/Characters/Player/Male/Medium/Bodies/M_MED_PawJasmineGravity/Skins/Greet/Materials/M_PawJasmineGravity_Parent
graph {
 properties {
  bHasPixelAnimation=true
  bUsedWithMeshDeformer=true
  bUsedWithMorphTargets=true
  bUsedWithNanite=true
  bUsedWithSkeletalMesh=true
  pixelDepthOffsetMode=PDOM_Legacy
  shadingModel=MSM_ClearCoat
  usedShadingModels=Clear Coat
 }
 nodes [
  n1 = MaterialFunctionCall() {materialFunction=&MF_CharacterParent_2023_PreFX}
  n2 = MaterialFunctionCall!(In=@n1.Result, Mask Input=@n1.Result) {materialFunction=/Game/Global/Materials/Character_Parent_2023/Hair_Parent_2023/MaterialFunctions/MF_FN_Hair_2023}
  n3 = BlendMaterialAttributes(
   A=@n1.Result
   B=StaticSwitchParameter(True=@n2.Result, False=@n1.Result) {defaultValue=true, group=Skin FX, parameterName=UseFNHair2_0}
   Alpha=@n2.HairMask
  )
  n4 = MaterialFunctionCall!(MaterialInput=@n3) {materialFunction=/Game/Global/Materials/Character_Parent_2023/Hair_Parent_2023/MaterialFunctions/MF_FlyAwayHair_2023}
  n5 = MaterialFunctionCall!(
   Tech Art Mask=@n1.Tech Art Mask
   Result=StaticSwitchParameter(True=StaticSwitchParameter(True=@n4.Result, False=@n3) {group=Skin FX, parameterName=IsFlyAwayHair?}, False=@n1.Result) {group=Paw Jasmine, parameterName=Use Hair}
  ) {materialFunction=&/MF_PawJasmine}
  GetMaterialAttributes(None=MaterialFunctionCall(Material=@n5."MA (Out)", M=@n1.M, Pre Skinned Local Position=@n1.M, Distance Blend=@n1.M, Fresnel=@n1.M) {materialFunction=&MF_CharacterParent_2023_PostFX}.Result) {attributeGetTypes [
    #1
    #2
    #3
    #4
    #5
    #6
    #7
    #8
    #9
    #10
    #11
    #12
    #13
   ]}
  n6 = NamedRerouteUsage()
  n7 = ScalarParameter() {defaultValue=0.5, parameterName=Clear Coat - Bottom Normal - Tiling}
  n8 = TextureSample(UVs=BumpOffset(Coordinate=Multiply(A=@n6, B=@n7), Height=ScalarParameter() {defaultValue=-1, parameterName=Param_3})) {
   samplerType=SAMPLERTYPE_Normal
   texture=/BRCosmetics/Characters/Player/Male/Medium/Bodies/M_MED_PawJasmineGravity/Skins/Greet/Textures/T_PawJasmine_Gravity_Greet_Tilable_N
  }
  LinearInterpolate(A=NamedRerouteUsage(), B=@n8.RGB, Alpha=NamedRerouteUsage())
  LinearInterpolate(Alpha=NamedRerouteUsage())
  Multiply(A=NamedRerouteUsage(), B=ScalarParameter() {defaultValue=1, group=Paw Jasmine, parameterName=Clear Coat - Outer Roughness})
  NamedRerouteDeclaration(None=@n5.Camo Area) {name=Camo Area, nodeColor=<rgba 0 0.0391 1 1>}
  NamedRerouteDeclaration(None=@n5.Roughness) {name=Roughness, nodeColor=<rgba 0 0.2969 1 1>}
 ]
}
